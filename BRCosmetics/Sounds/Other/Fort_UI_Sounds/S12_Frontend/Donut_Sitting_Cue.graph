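# Sound cue for the Season 12 frontend that loops the Donut_Sitting wave.
# A single looping wave player is the root, with no mixing or randomization.
# The cue overrides attenuation with a short falloff starting at the source and routes to the non-critical in-game UI sound class.
# It joins the UI_Donut_Computer_Duck_Bathroom_Conc concurrency set, which governs how it plays alongside other sounds in that set.
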SoundCue /BRCosmetics/Sounds/Other/Fort_UI_Sounds/S12_Frontend/Donut_Sitting_Cue
root=WavePlayer_1
nodes [{id=WavePlayer_1, properties {bLooping=true, soundWaveAssetPtr=&/Donut_Sitting}}]
properties {
 bOverrideAttenuation=true
 maxDistance=800
 soundClassObject=&Fort_UI_InGame_NonCritical
 volumeMultiplier=1
 attenuationOverrides {falloffDistance=800, attenuationShapeExtents {x=0}}
 concurrencySet [&/UI_Donut_Computer_Duck_Bathroom_Conc]
}
