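# Sound cue for a bubble foley effect in the Season 12 frontend UI sounds, named Dutchboy Bubbles.
# A Random node picks evenly among three BlowingBubbles emote foley waves, played at lowered pitch and raised volume.
# Attenuation is overridden off so it plays without distance falloff, and it routes through a first-person emote music sound class.
# A shared duck-bubbles concurrency set is assigned to limit overlapping playback.
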SoundCue /BRCosmetics/Sounds/Other/Fort_UI_Sounds/S12_Frontend/Dutchboy_Bubbles_Cue
root=Random_0
nodes [
 {id=Random_0, children [WavePlayer_0,WavePlayer_2,WavePlayer_3], properties {weights [1,1,1]}}
 {id=WavePlayer_0, properties {soundWaveAssetPtr=/Game/Athena/Sounds/Emotes/BlowingBubbles/Athena_Emote_BlowingBubbles_BubbleFoley_01}}
 {id=WavePlayer_2, properties {soundWaveAssetPtr=/Game/Athena/Sounds/Emotes/BlowingBubbles/Athena_Emote_BlowingBubbles_BubbleFoley_02}}
 {id=WavePlayer_3, properties {soundWaveAssetPtr=/Game/Athena/Sounds/Emotes/BlowingBubbles/Athena_Emote_BlowingBubbles_BubbleFoley_03}}
]
properties {
 bOverrideAttenuation=true
 maxDistance=2097152
 pitchMultiplier=0.7
 priority=0.9
 soundClassObject=/Game/Sounds/FortSoundClassesAndModes/Fort_Music_Emote_1P_PSM
 volumeMultiplier=1.3
 attenuationOverrides {bAttenuate=false, spatializationAlgorithm=SPATIALIZATION_HRTF}
 concurrencyOverrides {maxCount=1, resolutionRule=StopOldest}
 concurrencySet [&/UI_Donut_Computer_Duck_Bubbles_Conc]
}
