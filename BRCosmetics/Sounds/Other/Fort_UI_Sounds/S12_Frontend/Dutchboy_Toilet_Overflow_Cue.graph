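# Looping sound cue for the overflowing Dutchboy toilet in the Season 12 frontend, routed to the non-critical in-game UI sound class.
# A mixer layers a general water-flow loop with a slightly louder close-up loop.
# Each layer has its own HRTF-spatialized attenuation override, with a much shorter falloff on the close loop so that it adds detail only nearby.
# A shared concurrency set, named for computer ducking of the toilet, governs playback.
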SoundCue /BRCosmetics/Sounds/Other/Fort_UI_Sounds/S12_Frontend/Dutchboy_Toilet_Overflow_Cue
root=Mixer_0
nodes [
 {
  id=Attenuation_0
  children [WavePlayer_1]
  properties {
   bOverrideAttenuation=true
   attenuationOverrides {falloffDistance=1500, spatializationAlgorithm=SPATIALIZATION_HRTF, attenuationShapeExtents {x=0}}
  }
 }
 {
  id=Attenuation_1
  children [WavePlayer_2]
  properties {
   bOverrideAttenuation=true
   attenuationOverrides {falloffDistance=500, spatializationAlgorithm=SPATIALIZATION_HRTF, attenuationShapeExtents {x=100}}
  }
 }
 {id=Mixer_0, children [Attenuation_0, Attenuation_1], properties {inputVolume [1,1.2]}}
 {
  id=WavePlayer_1
  properties {bLooping=true, soundWaveAssetPtr=&/UI_S12_Dutchboy_Toilet_Flow_Loop}
 }
 {
  id=WavePlayer_2
  properties {bLooping=true, soundWaveAssetPtr=&/UI_S12_Dutchboy_Toilet_Close_Loop}
 }
]
properties {
 maxDistance=1500
 soundClassObject=&Fort_UI_InGame_NonCritical
 volumeMultiplier=1.75
 attenuationOverrides {falloffDistance=2000, attenuationShapeExtents {x=0}}
 concurrencyOverrides {bVolumeScaleCanRelease=true, maxCount=1, resolutionRule=StopOldest, volumeScaleAttackTime=0.25, volumeScaleReleaseTime=2}
 concurrencySet [&/UI_Donut_Computer_Duck_Toilet_Conc]
}
