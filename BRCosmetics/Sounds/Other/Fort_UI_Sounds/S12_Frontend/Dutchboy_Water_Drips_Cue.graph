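# Dutchboy_Water_Drips_Cue is a UI sound cue that plays a looping water-drip ambience, named for the Season 12 frontend.
# A single looping wave player feeds the root with UI_S12_Dutchboy_Drips_Loop at reduced volume, in the non-critical in-game UI sound class.
# Overridden attenuation disables spatialization and distance falloff, so it plays flat.
# It joins the UI_Donut_Computer_Duck_Water_Drips_Conc concurrency group, whose name suggests ducking around computer sounds.
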SoundCue /BRCosmetics/Sounds/Other/Fort_UI_Sounds/S12_Frontend/Dutchboy_Water_Drips_Cue
root=WavePlayer_0
nodes [{
  id=WavePlayer_0
  properties {bLooping=true, soundWaveAssetPtr=&/UI_S12_Dutchboy_Drips_Loop}
 }]
properties {
 bOverrideAttenuation=true
 maxDistance=2097152
 priority=0.8
 soundClassObject=&Fort_UI_InGame_NonCritical
 volumeMultiplier=0.45
 attenuationOverrides {bAttenuate=false, bSpatialize=false, falloffDistance=1000, attenuationShapeExtents {x=1000}}
 concurrencyOverrides {maxCount=1, resolutionRule=StopOldest, volumeScaleAttackTime=0.25, volumeScaleReleaseTime=2}
 concurrencySet [&/UI_Donut_Computer_Duck_Water_Drips_Conc]
}
