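# Dutchboy water-submerged cue for the Season 12 frontend, played as a non-spatialized, unattenuated UI sound.
# A mixer layers a random pick of three first-person player swim one-shots with a looping underwater bed that starts after a short fixed delay.
# It uses the non-critical in-game UI sound class and references a concurrency set named for computer-duck water drips.
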SoundCue /BRCosmetics/Sounds/Other/Fort_UI_Sounds/S12_Frontend/Dutchboy_Water_Submerged_Cue
root=Mixer_0
nodes [
 {id=Delay_0, children [WavePlayer_1], properties {delayMax=0.15, delayMin=0.15}}
 {id=Mixer_0, children [Delay_0,Random_1], properties {inputVolume [1,1.1]}}
 {id=Random_1, children [WavePlayer_5,WavePlayer_6,WavePlayer_9], properties {weights [1,1,1]}}
 {
  id=WavePlayer_1
  properties {bLooping=true, soundWaveAssetPtr=&/Dutchboy_Water_Underwater_Loop_01}
 }
 {id=WavePlayer_5, properties {soundWaveAssetPtr=/Game/Athena/Sounds/Water/Interaction/Water_Player_Swim_Normal_1P_01}}
 {id=WavePlayer_6, properties {soundWaveAssetPtr=/Game/Athena/Sounds/Water/Interaction/Water_Player_Swim_Normal_1P_02}}
 {id=WavePlayer_9, properties {soundWaveAssetPtr=/Game/Athena/Sounds/Water/Interaction/Water_Player_Swim_Normal_1P_06}}
]
properties {
 bOverrideAttenuation=true
 maxDistance=2097152
 priority=0.9
 soundClassObject=&Fort_UI_InGame_NonCritical
 volumeMultiplier=1
 attenuationOverrides {bAttenuate=false, bSpatialize=false, falloffDistance=1000, attenuationShapeExtents {x=1000}}
 concurrencyOverrides {maxCount=1, resolutionRule=StopOldest, volumeScaleAttackTime=0.25, volumeScaleReleaseTime=2}
 concurrencySet [&/UI_Donut_Computer_Duck_Water_Drips_Conc]
}
