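# Sound cue for the Meow speedbag sound, filed among the Season 12 frontend UI sounds.
# A single wave player feeds a modulator that slightly varies pitch and volume on each play.
# It plays at reduced volume through the non-critical in-game UI sound class, using item-interact attenuation for distance falloff.
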SoundCue /BRCosmetics/Sounds/Other/Fort_UI_Sounds/S12_Frontend/Meow_Speedbag_Cue
root=Modulator_0
nodes [
 {id=Modulator_0, children [WavePlayer_1], properties {pitchMax=1.02, pitchMin=0.98, volumeMax=1.1}}
 {id=WavePlayer_1, properties {soundWaveAssetPtr=&/Meow_Speedbag_01}}
]
properties {
 attenuationSettings=/Game/Sounds/Attenuation/Fort_Item_Interact_Att
 maxDistance=1850
 soundClassObject=&Fort_UI_InGame_NonCritical
 volumeMultiplier=0.45
}
