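# S12_UI_DiceGame_Cue is a UI sound cue for a dice game among the Season 12 frontend sounds.
# A single wave player feeds the photographer dice game wave straight to the output at reduced volume, routed through the Fort_UI sound class.
# Attenuation is overridden off so it plays non-spatially, and concurrency allows one instance, stopping the oldest when retriggered.
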SoundCue /BRCosmetics/Sounds/Other/Fort_UI_Sounds/S12_Frontend/S12_UI_DiceGame_Cue
root=WavePlayer_2
nodes [{id=WavePlayer_2, properties {soundWaveAssetPtr=&/UI_Photographer_DiceGame_01}}]
properties {
 bOverrideAttenuation=true
 maxDistance=2097152
 soundClassObject=&Fort_UI
 volumeMultiplier=0.7
 attenuationOverrides {bAttenuate=false, falloffDistance=0, attenuationShapeExtents {x=0}}
 concurrencyOverrides {maxCount=1, resolutionRule=StopOldest}
}
