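# Sound Cue for the looping pulsing electrical cable ambience emitter in the S12 frontend.
# A single looping Wave Player feeds the output directly at reduced volume, routed through the Fort_UI sound class.
# The cue overrides attenuation inline to set its falloff, shape extent, stereo spread and occlusion filter.
# It also carries single-instance, prevent-new concurrency settings, though the concurrency override flag is not set here.
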SoundCue /BRCosmetics/Sounds/Other/Fort_UI_Sounds/S12_Frontend/UI_AMBS12_Pulsing_Cable_Emitter_Cue
root=WavePlayer_1
nodes [{
  id=WavePlayer_1
  properties {bLooping=true, soundWaveAssetPtr=&/Amb_Emit_Elec_Pulsing_Cable_Loop_03}
 }]
properties {
 bOverrideAttenuation=true
 maxDistance=5512
 soundClassObject=&Fort_UI
 volumeMultiplier=0.5
 attenuationOverrides {falloffDistance=5000, occlusionLowPassFilterFrequency=1000, stereoSpread=256, attenuationShapeExtents {x=512}}
 concurrencyOverrides {maxCount=1, resolutionRule=PreventNew}
}
