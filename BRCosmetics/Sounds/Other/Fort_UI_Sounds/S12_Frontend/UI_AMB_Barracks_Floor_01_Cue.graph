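# UI_AMB_Barracks_Floor_01_Cue is a looping barracks ambience cue, filed with the Season 12 frontend UI sounds.
# A single looping wave player plays AMB_S12_Barracks_Loop at reduced volume through the Fort_UI sound class.
# Overridden attenuation disables spatialization but keeps a distance falloff beyond an inner radius.
# Concurrency settings specify one instance with PreventNew, though the concurrency override flag is not set.
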SoundCue /BRCosmetics/Sounds/Other/Fort_UI_Sounds/S12_Frontend/UI_AMB_Barracks_Floor_01_Cue
root=WavePlayer_0
nodes [{id=WavePlayer_0, properties {bLooping=true, soundWaveAssetPtr=&/AMB_S12_Barracks_Loop}}]
properties {
 bOverrideAttenuation=true
 maxDistance=1049.1934
 soundClassObject=&Fort_UI
 volumeMultiplier=0.4
 attenuationOverrides {bSpatialize=false, falloffDistance=301.4479, attenuationShapeExtents {x=747.7454}}
 concurrencyOverrides {maxCount=1, resolutionRule=PreventNew}
}
