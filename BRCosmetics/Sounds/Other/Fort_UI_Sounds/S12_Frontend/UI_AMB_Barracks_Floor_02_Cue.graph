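# Looping ambience Sound Cue, named for floor 02 of the Season 12 frontend Barracks, in the Fort_UI sound class.
# A single looping wave player reuses the Apollo yacht below-deck room bed at reduced volume.
# Inline attenuation overrides turn off spatialization and set a custom shape extent and falloff distance, so the bed plays without directional panning.
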SoundCue /BRCosmetics/Sounds/Other/Fort_UI_Sounds/S12_Frontend/UI_AMB_Barracks_Floor_02_Cue
root=WavePlayer_1
nodes [{
  id=WavePlayer_1
  properties {bLooping=true, soundWaveAssetPtr=/BRCosmetics/Sounds/Other/Apollo/POI/S12/AMB_Bed_Yacht_Room_Below_Deck_Loop_01}
 }]
properties {
 bOverrideAttenuation=true
 maxDistance=707.4496
 soundClassObject=&Fort_UI
 volumeMultiplier=0.65
 attenuationOverrides {bSpatialize=false, falloffDistance=301.4479, attenuationShapeExtents {x=406.0017}}
 concurrencyOverrides {maxCount=1, resolutionRule=PreventNew}
}
