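# Looping UI ambience cue, named for Barracks Floor 03 in the Season 12 frontend.
# A Mixer layers a quieter, slightly pitched-down S12 Armory loop, fed through a Modulator, under a boosted Heroes Tab ambience loop.
# It plays non-spatialized at reduced volume in the Fort_UI sound class.
# Concurrency allows a single instance, stopping the oldest when a new one starts.
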SoundCue /BRCosmetics/Sounds/Other/Fort_UI_Sounds/S12_Frontend/UI_AMB_Barracks_Floor_03_Cue
root=Mixer_0
nodes [
 {id=Mixer_0, children [Modulator_0,WavePlayer_3], properties {inputVolume [0.3,1.5]}}
 {id=Modulator_0, children [WavePlayer_4], properties {pitchMax=0.95}}
 {
  id=WavePlayer_3
  properties {bLooping=true, soundWaveAssetPtr=/BRCosmetics/Sounds/Other/Fort_Ambient_Looping/FrontEnd_Ambient/AMB_Heroes_Tab_01_Loop}
 }
 {id=WavePlayer_4, properties {bLooping=true, soundWaveAssetPtr=/Game/Sounds/Fort_UI_Sounds/S12_Frontend/AMB_S12_Armory_Loop}}
]
properties {
 bOverrideAttenuation=true
 bOverrideConcurrency=true
 maxDistance=1049.1934
 soundClassObject=&Fort_UI
 volumeMultiplier=0.5
 attenuationOverrides {bSpatialize=false, falloffDistance=301.4479, attenuationShapeExtents {x=747.7454}}
 concurrencyOverrides {maxCount=1, resolutionRule=StopOldest}
}
