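# Sound cue for a looping ambient bed in the frontend, named for Barracks floor 04, routed through the Fort_UI sound class.
# A single looping wave player feeds the root with a bedroom computer loop, boosted by a volume multiplier.
# An attenuation override turns off spatialization and sets its own falloff distance, so the bed is not panned by position.
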SoundCue /BRCosmetics/Sounds/Other/Fort_UI_Sounds/S12_Frontend/UI_AMB_Barracks_Floor_04_Cue
root=WavePlayer_1
nodes [{
  id=WavePlayer_1
  properties {bLooping=true, soundWaveAssetPtr=&/AMB_Bed_Bedroom_Computer_Loop_01}
 }]
properties {
 bOverrideAttenuation=true
 maxDistance=570.4617
 soundClassObject=&Fort_UI
 volumeMultiplier=1.8
 attenuationOverrides {bSpatialize=false, falloffDistance=407.1564, attenuationShapeExtents {x=163.3053}}
 concurrencyOverrides {maxCount=1, resolutionRule=PreventNew}
}
