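# UI_AMB_Barracks_Floor_05_Cue is a SoundCue that plays a looping frontend ambience for the Season 12 barracks, using the Midas room ambience wave.
# A single looping wave player feeds the output in the Fort_UI sound class with boosted volume.
# Overridden attenuation applies distance falloff without spatialization.
# Concurrency allows one instance and prevents new ones from starting.
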SoundCue /BRCosmetics/Sounds/Other/Fort_UI_Sounds/S12_Frontend/UI_AMB_Barracks_Floor_05_Cue
root=WavePlayer_1
nodes [{id=WavePlayer_1, properties {bLooping=true, soundWaveAssetPtr=&/S12_Midas_Room_Amb}}]
properties {
 bOverrideAttenuation=true
 maxDistance=960.9261
 soundClassObject=&Fort_UI
 volumeMultiplier=2
 attenuationOverrides {bSpatialize=false, falloffDistance=758.576, attenuationShapeExtents {x=202.3501}}
 concurrencyOverrides {maxCount=1, resolutionRule=PreventNew}
}
