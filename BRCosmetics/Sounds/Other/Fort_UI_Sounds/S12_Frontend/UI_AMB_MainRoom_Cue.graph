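# Sound cue for the main-room ambience of the Season 12 frontend lobby, playing on the Fort_UI sound class at reduced volume.
# A single looping wave player feeds AMB_S12_Lobby straight to the output.
# Overridden attenuation keeps the sound non-spatialized while configuring distance falloff and low-pass occlusion traced on a custom game channel.
# Its concurrency override settings specify one instance and prevent new ones.
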SoundCue /BRCosmetics/Sounds/Other/Fort_UI_Sounds/S12_Frontend/UI_AMB_MainRoom_Cue
root=WavePlayer_1
nodes [{id=WavePlayer_1, properties {bLooping=true, soundWaveAssetPtr=&/AMB_S12_Lobby}}]
properties {
 bOverrideAttenuation=true
 maxDistance=3437.6379
 soundClassObject=&Fort_UI
 volumeMultiplier=0.45
 attenuationOverrides {
  bEnableOcclusion=true
  bSpatialize=false
  falloffDistance=2333.8333
  occlusionLowPassFilterFrequency=1000
  occlusionTraceChannel=ECC_GameTraceChannel11
  attenuationShapeExtents {x=1103.8047}
 }
 concurrencyOverrides {maxCount=1, resolutionRule=PreventNew}
}
