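# UI sound cue for the armory door opening, in the Season 12 frontend sound folder.
# A single wave player plays the door-open wave at reduced volume in the Fort UI sound class.
# An attenuation override turns off distance attenuation.
# Concurrency comes from a shared three-maximum, stop-furthest rule; the stored single-instance override is not enabled.
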SoundCue /BRCosmetics/Sounds/Other/Fort_UI_Sounds/S12_Frontend/UI_Armory_Door_Open_Cue
root=WavePlayer_0
nodes [{id=WavePlayer_0, properties {soundWaveAssetPtr=&/UI_Armory_Door_Open}}]
properties {
 bOverrideAttenuation=true
 maxDistance=2097152
 soundClassObject=&Fort_UI
 volumeMultiplier=0.3
 attenuationOverrides {bAttenuate=false}
 concurrencyOverrides {maxCount=1, resolutionRule=StopOldest}
 concurrencySet [/Game/Sounds/Concurrency_Rules/ThreeMaximum_StopFurthest_VolumeScale_02]
}
