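# Sound cue for the barracks door opening sound in the Season 12 frontend UI.
# A single wave player feeds the output at reduced volume, routed to the Fort_UI sound class.
# The cue overrides attenuation with its own falloff settings.
# A shared three-maximum, stop-furthest concurrency set governs it, as the stored single-voice concurrency override is not switched on.
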SoundCue /BRCosmetics/Sounds/Other/Fort_UI_Sounds/S12_Frontend/UI_Barracks_Door_Open_Cue
root=WavePlayer_0
nodes [{id=WavePlayer_0, properties {soundWaveAssetPtr=&/UI_Barracks_Door_Open}}]
properties {
 bOverrideAttenuation=true
 maxDistance=3100
 soundClassObject=&Fort_UI
 volumeMultiplier=0.65
 attenuationOverrides {falloffDistance=3000, attenuationShapeExtents {x=100}}
 concurrencyOverrides {maxCount=1, resolutionRule=StopOldest}
 concurrencySet [/Game/Sounds/Concurrency_Rules/ThreeMaximum_StopFurthest_VolumeScale_02]
}
