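# UI sound cue that plays an elevator door closing sound for the Season 12 frontend, routed through the Fort UI sound class.
# A single wave player feeds the root directly, reusing the Barracks back door close wave at low volume and slightly lowered pitch.
# A concurrency override allows only one instance at a time, stopping the oldest when it retriggers.
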SoundCue /BRCosmetics/Sounds/Other/Fort_UI_Sounds/S12_Frontend/UI_Elevator_Door_Close_Cue
root=WavePlayer_1
nodes [{id=WavePlayer_1, properties {soundWaveAssetPtr=&/UI_S12_Barracks_Back_Door_Close}}]
properties {
 bOverrideConcurrency=true
 maxDistance=2097152
 pitchMultiplier=0.95
 soundClassObject=&Fort_UI
 volumeMultiplier=0.1
 concurrencyOverrides {maxCount=1, resolutionRule=StopOldest}
}
