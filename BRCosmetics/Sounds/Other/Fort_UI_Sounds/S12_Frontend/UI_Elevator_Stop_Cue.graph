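# UI_Elevator_Stop_Cue is a Sound Cue in the Season 12 frontend UI sound set, named for an elevator stop.
# It has no nodes and a null root, so as defined here it plays no audio.
# It keeps only playback settings: the Fort_UI sound class and a concurrency override that allows one instance at a time, stopping the oldest.
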SoundCue /BRCosmetics/Sounds/Other/Fort_UI_Sounds/S12_Frontend/UI_Elevator_Stop_Cue
root=null
nodes []
properties {
 bOverrideConcurrency=true
 maxDistance=2097152
 soundClassObject=&Fort_UI
 concurrencyOverrides {maxCount=1, resolutionRule=StopOldest}
}
