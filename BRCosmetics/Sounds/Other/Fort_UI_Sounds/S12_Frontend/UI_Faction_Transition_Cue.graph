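# UI sound cue for a faction transition in the frontend, judging by its name and folder, played quietly in the UI sound class.
# A mixer layers a pitched-up hero-appear teleport sound with one of three randomly chosen class-selector change sounds, each branch slightly delayed, the selector sound arriving first.
# An outer modulator adds slight random pitch variation.
# Overridden concurrency caps simultaneous plays, stopping the oldest.
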SoundCue /BRCosmetics/Sounds/Other/Fort_UI_Sounds/S12_Frontend/UI_Faction_Transition_Cue
root=Modulator_0
nodes [
 {id=Delay_0, children [Random_1], properties {delayMax=0.04, delayMin=0.04}}
 {id=Delay_1, children [Modulator_2], properties {delayMax=0.1, delayMin=0.1}}
 {id=Mixer_0, children [Delay_1,Delay_0,null], properties {inputVolume [0.6,1,0.5]}}
 {id=Modulator_0, children [Mixer_0], properties {pitchMax=1.02, pitchMin=0.98}}
 {id=Modulator_2, children [WavePlayer_9], properties {pitchMax=1.3, pitchMin=1.3}}
 {id=Random_1, children [WavePlayer_0,WavePlayer_5,WavePlayer_6], properties {weights [1,1,1]}}
 {id=WavePlayer_0, properties {soundWaveAssetPtr=/BRCosmetics/Sounds/Other/Creative/Toys/ClassSelector/Class_Selector_Change_03}}
 {id=WavePlayer_5, properties {soundWaveAssetPtr=/Game/Sounds/Creative/Toys/ClassSelector/Class_Selector_Change_01}}
 {id=WavePlayer_6, properties {soundWaveAssetPtr=/Game/Sounds/Creative/Toys/ClassSelector/Class_Selector_Change_02}}
 {id=WavePlayer_9, properties {soundWaveAssetPtr=/Game/Sounds/Fort_UI_Sounds/FrontEnd/New_FrontEnd/UI_Teleport_Hero_Appear_02}}
]
properties {
 bOverrideAttenuation=true
 bOverrideConcurrency=true
 maxDistance=1000
 soundClassObject=&Fort_UI
 volumeMultiplier=0.08
 attenuationOverrides {falloffDistance=1000, attenuationShapeExtents {x=0}}
 concurrencyOverrides {maxCount=2, resolutionRule=StopOldest, volumeScale=0.7}
}
