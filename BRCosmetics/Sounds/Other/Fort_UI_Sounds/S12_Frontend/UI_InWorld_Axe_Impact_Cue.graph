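# Sound cue for an axe impact in the Season 12 frontend UI sound set, played through the Fort_UI sound class.
# A fixed short delay node leads into a single wave player for a cleaver wall impact sample.
# Overridden attenuation turns off distance falloff, and the concurrency override settings specify a prevent-new resolution rule.
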SoundCue /BRCosmetics/Sounds/Other/Fort_UI_Sounds/S12_Frontend/UI_InWorld_Axe_Impact_Cue
root=Delay_0
nodes [
 {id=Delay_0, children [WavePlayer_1], properties {delayMax=0.03, delayMin=0.03}}
 {id=WavePlayer_1, properties {soundWaveAssetPtr=&/Cleaver_Wall_Impact_01}}
]
properties {
 bOverrideAttenuation=true
 maxDistance=2097152
 soundClassObject=&Fort_UI
 attenuationOverrides {bAttenuate=false}
 concurrencyOverrides {resolutionRule=PreventNew}
}
