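# UI_InWorld_Paper_Foley_Cue is a Sound Cue for a paper foley sound in the Season 12 frontend UI sound set.
# A single wave player feeds UI_Inworld_Paper_Foley_01 straight to the output, routed through the Fort_UI sound class at slightly reduced volume.
# An attenuation override turns off distance falloff, and a concurrency override uses PreventNew so new instances are rejected once the limit is reached.
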SoundCue /BRCosmetics/Sounds/Other/Fort_UI_Sounds/S12_Frontend/UI_InWorld_Paper_Foley_Cue
root=WavePlayer_0
nodes [{id=WavePlayer_0, properties {soundWaveAssetPtr=&/UI_Inworld_Paper_Foley_01}}]
properties {
 bOverrideAttenuation=true
 maxDistance=2097152
 soundClassObject=&Fort_UI
 volumeMultiplier=0.9
 attenuationOverrides {bAttenuate=false}
 concurrencyOverrides {resolutionRule=PreventNew}
}
