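# Sound cue for in-world punching bag impacts in the Season 12 frontend, routed to the Fort_UI sound class.
# A Modulator feeds a quality switch whose first two levels randomly pick from four impact waves and last two from two.
# The two-wave random node is excluded from branch culling.
# Overridden attenuation turns off volume falloff but enables occlusion, and the cue's debug flag is on.
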SoundCue /BRCosmetics/Sounds/Other/Fort_UI_Sounds/S12_Frontend/UI_InWorld_PunchingBag_Cue
root=Modulator_0
nodes [
 {id=Modulator_0, children [QualityLevel_0]}
 {id=QualityLevel_0, children [Random_0,Random_0,Random_1,Random_1]}
 {id=Random_0, children [WavePlayer_1,WavePlayer_2,WavePlayer_3,WavePlayer_4], properties {weights [1,1,1,1]}}
 {id=Random_1, children [WavePlayer_5,WavePlayer_6], properties {bShouldExcludeFromBranchCulling=true, weights [1,1]}}
 {id=WavePlayer_1, properties {soundWaveAssetPtr=&/PunchingBag_Impact01}}
 {id=WavePlayer_2, properties {soundWaveAssetPtr=&/PunchingBag_Impact02}}
 {id=WavePlayer_3, properties {soundWaveAssetPtr=&/PunchingBag_Impact03}}
 {id=WavePlayer_4, properties {soundWaveAssetPtr=&/PunchingBag_Impact04}}
 {id=WavePlayer_5, properties {soundWaveAssetPtr=&PunchingBag_Impact01}}
 {id=WavePlayer_6, properties {soundWaveAssetPtr=&PunchingBag_Impact02}}
]
properties {
 bDebug=true
 bOverrideAttenuation=true
 maxDistance=2097152
 pitchMultiplier=0.97
 soundClassObject=&Fort_UI
 attenuationOverrides {
  bAttenuate=false
  bEnableOcclusion=true
  falloffDistance=6000
  occlusionLowPassFilterFrequency=2000
  occlusionTraceChannel=ECC_GameTraceChannel11
  attenuationShapeExtents {x=0, z=20}
 }
 concurrencyOverrides {bLimitToOwner=true, maxCount=3, resolutionRule=StopOldest}
}
