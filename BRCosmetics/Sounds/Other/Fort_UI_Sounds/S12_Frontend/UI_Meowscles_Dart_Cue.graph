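# Sound cue for the Season 12 frontend UI Meowscles dart sound, routed to the Fort_UI sound class.
# A single wave player feeds the UI_S12_Meowscles_Dart wave straight to the output, with no randomization or modulation.
# An attenuation override turns off distance attenuation but leaves occlusion enabled.
# The debug flag is on, and owner-limited concurrency settings are stored without the concurrency override flag set.
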SoundCue /BRCosmetics/Sounds/Other/Fort_UI_Sounds/S12_Frontend/UI_Meowscles_Dart_Cue
root=WavePlayer_0
nodes [{id=WavePlayer_0, properties {soundWaveAssetPtr=&/UI_S12_Meowscles_Dart}}]
properties {
 bDebug=true
 bOverrideAttenuation=true
 maxDistance=2097152
 soundClassObject=&Fort_UI
 attenuationOverrides {
  bAttenuate=false
  bEnableOcclusion=true
  falloffDistance=6000
  occlusionLowPassFilterFrequency=2000
  occlusionTraceChannel=ECC_GameTraceChannel11
  attenuationShapeExtents {x=0, z=20}
 }
 concurrencyOverrides {bLimitToOwner=true, maxCount=3, resolutionRule=StopOldest}
}
