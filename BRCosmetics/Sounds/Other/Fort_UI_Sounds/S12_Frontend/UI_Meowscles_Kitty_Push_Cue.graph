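# UI sound cue for the Season 12 frontend, named for a Meowscles kitty toy push.
# A Random node picks between two equally weighted push waves and feeds a Modulator that applies a barely perceptible pitch variation.
# It plays in the Fort_UI sound class, with attenuation overrides that turn off distance volume falloff but leave occlusion enabled.
# The cue's debug flag is on.
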SoundCue /BRCosmetics/Sounds/Other/Fort_UI_Sounds/S12_Frontend/UI_Meowscles_Kitty_Push_Cue
root=Modulator_0
nodes [
 {id=Modulator_0, children [Random_0], properties {pitchMax=1.001, pitchMin=0.999}}
 {id=Random_0, children [WavePlayer_4,WavePlayer_0], properties {weights [1,1]}}
 {id=WavePlayer_0, properties {soundWaveAssetPtr=&/UI_S12_Meowscles_Toy_Push_02}}
 {id=WavePlayer_4, properties {soundWaveAssetPtr=&/UI_S12_Meowscles_Toy_Push}}
]
properties {
 bDebug=true
 bOverrideAttenuation=true
 maxDistance=2097152
 soundClassObject=&Fort_UI
 attenuationOverrides {
  bAttenuate=false
  bEnableOcclusion=true
  falloffDistance=6000
  occlusionLowPassFilterFrequency=2000
  occlusionTraceChannel=ECC_GameTraceChannel11
  attenuationShapeExtents {x=0, z=20}
 }
 concurrencyOverrides {bLimitToOwner=true, maxCount=3, resolutionRule=StopOldest}
}
