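# Sound cue for the Midas phonograph crank sound in the Season 12 frontend UI, routed through the Fort_UI sound class.
# It randomly picks one of two equally weighted crank waves, applies slight pitch variation, and plays it after a short fixed delay.
# Overridden attenuation gives it distance falloff with HRTF spatialization.
# Overridden concurrency caps it at two instances and blocks new ones beyond that.
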SoundCue /BRCosmetics/Sounds/Other/Fort_UI_Sounds/S12_Frontend/UI_Midas_Phonograph_Crank_Cue
root=Delay_0
nodes [
 {id=Delay_0, children [Modulator_0], properties {delayMax=0.08, delayMin=0.08}}
 {id=Modulator_0, children [Random_0], properties {pitchMax=1.02, pitchMin=0.98}}
 {id=Random_0, children [WavePlayer_0,WavePlayer_2], properties {weights [1,1]}}
 {id=WavePlayer_0, properties {soundWaveAssetPtr=&/Phonograph_Crank_01}}
 {id=WavePlayer_2, properties {soundWaveAssetPtr=&/Phonograph_Crank_02}}
]
properties {
 bOverrideAttenuation=true
 bOverrideConcurrency=true
 maxDistance=3500
 soundClassObject=&Fort_UI
 volumeMultiplier=1.1
 attenuationOverrides {falloffDistance=2000, spatializationAlgorithm=SPATIALIZATION_HRTF, attenuationShapeExtents {x=1500}}
 concurrencyOverrides {maxCount=2, resolutionRule=PreventNew}
}
