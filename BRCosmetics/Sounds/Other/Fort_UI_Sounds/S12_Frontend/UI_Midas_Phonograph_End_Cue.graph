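# Sound cue for the end sound of the Midas phonograph, filed among the Season 12 frontend UI sounds.
# A fixed short delay leads into a single wave player for the phonograph end wave.
# It uses a first-person emote music sound class, with its own distance attenuation and a concurrency cap that blocks new instances once the limit is reached.
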SoundCue /BRCosmetics/Sounds/Other/Fort_UI_Sounds/S12_Frontend/UI_Midas_Phonograph_End_Cue
root=Delay_0
nodes [
 {id=Delay_0, children [WavePlayer_3], properties {delayMax=0.1, delayMin=0.1}}
 {id=WavePlayer_3, properties {soundWaveAssetPtr=&/S12_MIdas_Phongraph_End}}
]
properties {
 bOverrideAttenuation=true
 bOverrideConcurrency=true
 maxDistance=2500
 soundClassObject=/Game/Sounds/FortSoundClassesAndModes/Fort_Music_Emote_1P_PSM
 volumeMultiplier=1
 attenuationOverrides {falloffDistance=1000, attenuationShapeExtents {x=1500}}
 concurrencyOverrides {maxCount=2, resolutionRule=PreventNew}
}
