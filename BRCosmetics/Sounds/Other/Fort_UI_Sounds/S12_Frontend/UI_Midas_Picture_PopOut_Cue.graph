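# UI sound cue for the Season 12 frontend, named for the Midas picture pop-out.
# A single wave player plays the S12_MIdas_Picture_End wave at slightly reduced volume in the Fort_UI sound class.
# It overrides attenuation with HRTF spatialization and a distance falloff, and overrides concurrency to cap simultaneous instances, preventing new ones once the limit is reached.
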SoundCue /BRCosmetics/Sounds/Other/Fort_UI_Sounds/S12_Frontend/UI_Midas_Picture_PopOut_Cue
root=WavePlayer_1
nodes [{id=WavePlayer_1, properties {soundWaveAssetPtr=&/S12_MIdas_Picture_End}}]
properties {
 bOverrideAttenuation=true
 bOverrideConcurrency=true
 maxDistance=5000
 soundClassObject=&Fort_UI
 volumeMultiplier=0.85
 attenuationOverrides {falloffDistance=2000, spatializationAlgorithm=SPATIALIZATION_HRTF, attenuationShapeExtents {x=3000}}
 concurrencyOverrides {maxCount=2, resolutionRule=PreventNew}
}
