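# UI sound cue for a paper doll part transition in the frontend, judging by its name, playing the Season 13 cosmetic appear wave.
# A single wave player feeds a root Modulator that slightly randomizes pitch on each play.
# It routes to the Fort_UI sound class at reduced volume, with its own attenuation and concurrency overrides that allow one instance, stopping the oldest when retriggered.
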SoundCue /BRCosmetics/Sounds/Other/Fort_UI_Sounds/S12_Frontend/UI_PaperDoll_Part_Transition_Cue
root=Modulator_0
nodes [
 {id=Modulator_0, children [WavePlayer_2], properties {pitchMax=1.02, pitchMin=0.98}}
 {id=WavePlayer_2, properties {soundWaveAssetPtr=/BRCosmetics/Sounds/Other/Fort_UI_Sounds/S13_Frontend/S13_Cosmetic_Appear_Res}}
]
properties {
 bOverrideAttenuation=true
 bOverrideConcurrency=true
 maxDistance=800
 soundClassObject=&Fort_UI
 volumeMultiplier=0.4
 attenuationOverrides {falloffDistance=800, attenuationShapeExtents {x=0}}
 concurrencyOverrides {bVolumeScaleCanRelease=true, maxCount=1, resolutionRule=StopOldest, volumeScale=0.7}
}
