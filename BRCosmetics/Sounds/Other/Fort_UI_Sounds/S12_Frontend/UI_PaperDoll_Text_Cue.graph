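# UI sound cue for paper-doll text in the frontend, apparently a hover sound judging by the waves and concurrency asset it references.
# It layers a button hover wave and a hover-click wave in a mixer at different levels, passed through a modulator with fixed volume.
# It plays in the Fort UI sound class at reduced volume and references a shared text-hover concurrency set.
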SoundCue /BRCosmetics/Sounds/Other/Fort_UI_Sounds/S12_Frontend/UI_PaperDoll_Text_Cue
root=Modulator_0
nodes [
 {id=Mixer_0, children [WavePlayer_3,WavePlayer_2], properties {inputVolume [0.75,0.5]}}
 {id=Modulator_0, children [Mixer_0], properties {volumeMax=1, volumeMin=1}}
 {id=WavePlayer_2, properties {soundWaveAssetPtr=/Game/Sounds/Fort_UI_Sounds/FrontEnd/New_FrontEnd/UI_FN_Button_Hover}}
 {id=WavePlayer_3, properties {soundWaveAssetPtr=/Game/Sounds/Fort_UI_Sounds/FrontEnd/New_FrontEnd/UI_FN_Button_Hover_Click_02}}
]
properties {
 maxDistance=2097152
 soundClassObject=&Fort_UI
 volumeMultiplier=0.7
 concurrencyOverrides {maxCount=1, resolutionRule=PreventNew}
 concurrencySet [/Game/Sounds/Fort_UI_Sounds/S13_Frontend/UI_Frontend_Text_Hover_Conc]
}
