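# Sound cue for the Season 12 frontend UI, named for the completion of a Skye star.
# A single wave player feeds S12_Skye_Stars_End straight to the output through the Fort_UI sound class, with a slight volume boost.
# Its own concurrency override allows two simultaneous instances and blocks new ones beyond that.
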SoundCue /BRCosmetics/Sounds/Other/Fort_UI_Sounds/S12_Frontend/UI_Skye_Star_Complete_Cue
root=WavePlayer_0
nodes [{id=WavePlayer_0, properties {soundWaveAssetPtr=&/S12_Skye_Stars_End}}]
properties {
 bOverrideConcurrency=true
 maxDistance=2097152
 soundClassObject=&Fort_UI
 volumeMultiplier=1.25
 concurrencyOverrides {maxCount=2, resolutionRule=PreventNew}
}
