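# UI sound cue for interacting with a star in the Season 12 frontend, apparently the Skye stars given the wave names.
# A Random node picks one of three star sound variations with equal weight.
# It plays in the Fort_UI sound class with a slight volume boost.
# A concurrency override caps it at two instances, stopping the oldest with a short volume-scale release.
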SoundCue /BRCosmetics/Sounds/Other/Fort_UI_Sounds/S12_Frontend/UI_Syke_Star_Interact_Cue
root=Random_0
nodes [
 {id=Random_0, children [WavePlayer_0,WavePlayer_1,WavePlayer_2], properties {weights [1,1,1]}}
 {id=WavePlayer_0, properties {soundWaveAssetPtr=&/S12_Skye_Stars_01}}
 {id=WavePlayer_1, properties {soundWaveAssetPtr=/Game/Sounds/Fort_UI_Sounds/S12_Frontend/S12_Skye_Stars_02}}
 {id=WavePlayer_2, properties {soundWaveAssetPtr=&/S12_Skye_Stars_03}}
]
properties {
 bOverrideConcurrency=true
 maxDistance=2097152
 soundClassObject=&Fort_UI
 volumeMultiplier=1.2
 concurrencyOverrides {bVolumeScaleCanRelease=true, maxCount=2, resolutionRule=StopOldest, volumeScaleReleaseTime=0.2}
}
