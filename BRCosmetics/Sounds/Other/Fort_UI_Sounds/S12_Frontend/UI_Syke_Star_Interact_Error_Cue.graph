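# UI sound cue for an error when interacting with a Skye star, apparently in the Season 12 frontend.
# A single wave player feeds the S12_Skye_Star_Error wave through a default Modulator for slight pitch and volume variation.
# The cue uses the Fort_UI sound class with slightly lowered pitch and volume.
# Its own concurrency override caps simultaneous plays at two, stopping the oldest.
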SoundCue /BRCosmetics/Sounds/Other/Fort_UI_Sounds/S12_Frontend/UI_Syke_Star_Interact_Error_Cue
root=Modulator_0
nodes [
 {id=Modulator_0, children [WavePlayer_3]}
 {id=WavePlayer_3, properties {soundWaveAssetPtr=&/S12_Skye_Star_Error}}
]
properties {
 bOverrideConcurrency=true
 maxDistance=2097152
 pitchMultiplier=0.95
 soundClassObject=&Fort_UI
 volumeMultiplier=0.9
 concurrencyOverrides {bVolumeScaleCanRelease=true, maxCount=2, resolutionRule=StopOldest, volumeScaleReleaseTime=0.2}
}
