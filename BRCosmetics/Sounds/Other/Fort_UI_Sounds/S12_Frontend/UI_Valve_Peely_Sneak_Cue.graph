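# UI Sound Cue in the Season 12 frontend sound set, named for a Peely sneak moment, that reuses the Winterfest "Peely Rises Again" wave.
# A fixed short delay feeds a Modulator with pinned volume, which plays that single wave at reduced cue volume in the Fort_UI sound class.
# Overrides switch off distance attenuation and limit playback to one instance, stopping the oldest.
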SoundCue /BRCosmetics/Sounds/Other/Fort_UI_Sounds/S12_Frontend/UI_Valve_Peely_Sneak_Cue
root=Delay_3
nodes [
 {id=Delay_3, children [Modulator_0], properties {delayMax=0.9, delayMin=0.9}}
 {id=Modulator_0, children [WavePlayer_8], properties {volumeMax=1, volumeMin=1}}
 {id=WavePlayer_8, properties {soundWaveAssetPtr=/BRCosmetics/Sounds/Other/Events/Winterfest/Discoverables/Winterfest_Peely_Rises_Again}}
]
properties {
 bOverrideAttenuation=true
 maxDistance=2097152
 soundClassObject=&Fort_UI
 volumeMultiplier=0.45
 attenuationOverrides {bAttenuate=false, falloffDistance=800, spatializationAlgorithm=SPATIALIZATION_HRTF}
 concurrencyOverrides {maxCount=1, resolutionRule=StopOldest}
}
