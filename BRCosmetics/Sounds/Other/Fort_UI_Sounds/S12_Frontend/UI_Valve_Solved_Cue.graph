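# UI sound cue in the Season 12 frontend set, named for solving a valve.
# A single wave player feeds the root directly, reusing the Infiltration LTM "Intel_Downloaded_By_Enemy" wave at reduced volume.
# It routes through the Fort_UI sound class, and its attenuation override disables distance falloff.
# Its concurrency override values specify one instance with stop-oldest resolution, though no flag enabling that override is shown.
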SoundCue /BRCosmetics/Sounds/Other/Fort_UI_Sounds/S12_Frontend/UI_Valve_Solved_Cue
root=WavePlayer_1
nodes [{id=WavePlayer_1, properties {soundWaveAssetPtr=/BRCosmetics/Sounds/Other/LTM/Infiltration/Intel_Downloaded_By_Enemy}}]
properties {
 bOverrideAttenuation=true
 maxDistance=2097152
 soundClassObject=&Fort_UI
 volumeMultiplier=0.45
 attenuationOverrides {bAttenuate=false, falloffDistance=800, spatializationAlgorithm=SPATIALIZATION_HRTF}
 concurrencyOverrides {maxCount=1, resolutionRule=StopOldest}
}
