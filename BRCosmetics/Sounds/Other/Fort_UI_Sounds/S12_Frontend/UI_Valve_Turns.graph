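# UI sound cue for a valve turn in the Season 12 frontend, played in the Fort_UI sound class.
# A mixer layers a randomly chosen valve-turn sample, one of four equally weighted waves, with a steam sound.
# Each layer passes through its own modulator for variation.
# A concurrency override caps simultaneous plays at two and stops the oldest.
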SoundCue /BRCosmetics/Sounds/Other/Fort_UI_Sounds/S12_Frontend/UI_Valve_Turns
root=Mixer_0
nodes [
 {id=Mixer_0, children [Modulator_0,Modulator_1], properties {inputVolume [1,1]}}
 {id=Modulator_0, children [Random_0], properties {pitchMax=1.02}}
 {id=Modulator_1, children [WavePlayer_0]}
 {id=Random_0, children [WavePlayer_1,WavePlayer_2,WavePlayer_3,WavePlayer_4], properties {weights [1,1,1,1]}}
 {id=WavePlayer_0, properties {soundWaveAssetPtr=&/UI_Valve_Steam}}
 {id=WavePlayer_1, properties {soundWaveAssetPtr=&/UI_Valve_Turn_01}}
 {id=WavePlayer_2, properties {soundWaveAssetPtr=&/UI_Valve_Turn_02}}
 {id=WavePlayer_3, properties {soundWaveAssetPtr=&/UI_Valve_Turn_03}}
 {id=WavePlayer_4, properties {soundWaveAssetPtr=&/UI_Valve_Turn_04}}
]
properties {
 bOverrideConcurrency=true
 maxDistance=2097152
 soundClassObject=&Fort_UI
 volumeMultiplier=0.5
 concurrencyOverrides {bVolumeScaleCanRelease=true, maxCount=2, resolutionRule=StopOldest, volumeScaleReleaseTime=0.2}
}
