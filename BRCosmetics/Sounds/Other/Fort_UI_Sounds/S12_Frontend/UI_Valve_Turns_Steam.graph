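# UI sound cue for a steam valve turn in the Season 12 frontend, routed through the Fort_UI sound class.
# It picks one of two valve-turn steam waves at equal weight, then applies slight random pitch variation through a modulator.
# Its own concurrency override caps it at two simultaneous instances and blocks new plays beyond that, with reduced volume and a very large max distance.
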SoundCue /BRCosmetics/Sounds/Other/Fort_UI_Sounds/S12_Frontend/UI_Valve_Turns_Steam
root=Modulator_0
nodes [
 {id=Modulator_0, children [Random_1], properties {pitchMax=1.02, pitchMin=0.99}}
 {id=Random_1, children [WavePlayer_0,WavePlayer_3], properties {weights [1,1]}}
 {id=WavePlayer_0, properties {soundWaveAssetPtr=&/UI_Valve_Turn_Steam_02}}
 {id=WavePlayer_3, properties {soundWaveAssetPtr=&/UI_Valve_Turn_Steam_01}}
]
properties {
 bOverrideConcurrency=true
 maxDistance=2097152
 soundClassObject=&Fort_UI
 volumeMultiplier=0.65
 concurrencyOverrides {maxCount=2, resolutionRule=PreventNew}
}
